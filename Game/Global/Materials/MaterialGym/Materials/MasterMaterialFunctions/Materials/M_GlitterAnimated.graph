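# M_GlitterAnimated is a gym material that showcases the animated glitter effect used by weapon skins.
# It is a Default Lit material using material attributes, built around a single MF_GlitterFX function call.
# An empty SetMaterialAttributes feeds the function's input, and a constant white vector fills SkinFX_Mask.
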Material /Game/Global/Materials/MaterialGym/Materials/MasterMaterialFunctions/Materials/M_GlitterAnimated
graph {
 properties {bUseMaterialAttributes=true, pixelDepthOffsetMode=PDOM_Legacy, usedShadingModels=Default Lit}
 nodes [MaterialFunctionCall!(In=SetMaterialAttributes!(MaterialAttributes), SkinFX_Mask=Constant4Vector() {constant=<rgba 1 1 1 1>}.RGBA) {materialFunction=/Game/Weapons/WeaponSkins/MaterialFunction/MF_GlitterFX}]
}
