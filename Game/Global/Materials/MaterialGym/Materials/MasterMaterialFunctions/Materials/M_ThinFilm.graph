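# M_ThinFilm is a Default Lit material in the MaterialGym folder that wraps the weapon-skin ThinFilmFunction, apparently to showcase that effect in isolation.
# A Set Material Attributes node fills only base color, from a BaseColor vector parameter that defaults to black.
# Those attributes feed ThinFilmFunction with a constant full alpha, and the material uses the Material Attributes workflow.
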Material /Game/Global/Materials/MaterialGym/Materials/MasterMaterialFunctions/Materials/M_ThinFilm
graph {
 properties {bUseMaterialAttributes=true, pixelDepthOffsetMode=PDOM_Legacy, usedShadingModels=Default Lit}
 nodes [MaterialFunctionCall!(
   Alpha=Constant() {r=1}
   Material Attributes=SetMaterialAttributes!(MaterialAttributes, Base Color=VectorParameter() {parameterName=BaseColor, defaultValue=<rgba 0 0 0 1>}.RGB) {attributeSetTypes [#1]}
  ) {materialFunction=/Game/Weapons/WeaponSkins/MaterialFunction/ThinFilmFunction}]
}
